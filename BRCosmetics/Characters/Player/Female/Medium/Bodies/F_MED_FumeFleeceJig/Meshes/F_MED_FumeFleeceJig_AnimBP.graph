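# Animation Blueprint for the female medium FumeFleeceJig body cosmetic, adding deformation and secondary motion to the pose copied from the attached parent mesh.
# A deform Control Rig takes a parachute-arms input, then a weighted pelvis/spine constraint drives the holster bone before RigidBody simulates in spine_01 space.
# In the front end, the Frontend_Idle_Bypass curve blends DefaultSlot over the holster branch during custom idles.
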AnimBlueprint /BRCosmetics/Characters/Player/Female/Medium/Bodies/F_MED_FumeFleeceJig/Meshes/F_MED_FumeFleeceJig_AnimBP
target_skeleton=&/F_MED_FumeFleeceJig_Skeleton
graphs [
 AnimGraph {nodes [
   n1 = Control Rig!(Source=Copy Pose From Mesh() {bUseAttachedParent=true}, Alpha, isParachuting:double=Get isParachuteArms() -> (isParachuteArms:double)) -> (Pose) {
    alphaBoolBlend {blendOption=Linear}
    controlRigAssetReference {blueprintRigClass=&/F_MED_FumeFleeceJig_CtrlRig_Deform.F_MED_FumeFleeceJig_CtrlRig_Deform_C}
   }
   n2 = "Constraint - dyn_holster_MSTR (PPP)"!(
    ComponentPose=Local To Component(LocalPose=@n1)
    Alpha=1
    ConstraintWeights_0=0.831
    ConstraintWeights_1:float=0.146
    ConstraintWeights_2:float=0.023
   ) -> (Pose) {
    alphaBoolBlend {blendOption=Linear}
    boneToModify {boneName=dyn_holster_MSTR}
    constraintSetup [{targetBone {boneName=pelvis}}, {targetBone {boneName=spine_01}}, {targetBone {boneName=spine_02}}]
    constraintWeights [1,1,1]
   }
   n3 = RigidBody!(ComponentPose=@n2, Alpha, OverrideWorldGravity:Vector="0,0,0", ExternalForce) -> (Pose) {
    bClampLinearTranslationLimitToRefPose=true
    overridePhysicsAsset=&/F_MED_FumeFleeceJig_Physics
    simulationSpace=BaseBoneSpace
    alphaBoolBlend {blendOption=Linear}
    baseBoneRef {boneName=spine_01}
    componentAppliedLinearAccClamp=<3000 3000 3000>
    simSpaceSettings {
     maxAngularAcceleration=14
     maxAngularVelocity=6
     maxLinearAcceleration=1150
     maxLinearVelocity=450
     worldAlpha=1
     externalLinearDragV=<2 2 4>
    }
   }
   Output Pose(Result=Layered blend per bone(
     BasePose=Component To Local(ComponentPose=@n3)
     BlendPoses_0=Slot 'DefaultSlot'()
     BlendWeights_0=Get Custom Idle is Playing() -> (Custom Idle is Playing:double)
    ) {layerSetup [{branchFilters [{boneName=dyn_holster_MSTR}]}]}) {layerGroup=None}
  ]}
 EventGraph {nodes [
   Event BlueprintUpdateAnimation()
   > Branch(Condition=Get bIsInFrontEnd())
   > Set Custom Idle is Playing(Custom Idle is Playing:double=GetCurveValue(self=GetOwnerFrontEndAnimInstance(), CurveName=Frontend_Idle_Bypass)) -> (Output_Get:double)
  ]}
]
variables [
 Custom Idle is Playing {type=number}
 isParachuteArms {type=number}
]
